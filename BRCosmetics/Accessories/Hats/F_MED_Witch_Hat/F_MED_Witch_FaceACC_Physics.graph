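# Physics asset for the female medium witch hat face accessory, giving the hat and hair secondary motion on the head.
# Box bodies cover the head, three hat segments, two rim bones and three hair bones.
# The hat chains from a tightly limited base at the head through looser upper segments, with rims limited mostly to twist.
# Hair pieces hang directly from the head, each with its own swing limits.
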PhysicsAsset /BRCosmetics/Accessories/Hats/F_MED_Witch_Hat/F_MED_Witch_FaceACC_Physics
bodies {
 dyn_hair_back [{shapeType=Box, extentX=8.112, extentY=8.112, extentZ=8.112, center=<-6.3968 -4.6953 -0.1665>}]
 dyn_hair_l [{shapeType=Box, extentX=5.1366, extentY=5.1366, extentZ=5.1366, center=<2.8749 -0.0229 0.377>}]
 dyn_hair_r [{
   shapeType=Box
   extentX=5
   extentY=5
   extentZ=3
   shapeName=dyn_hair_r_box
   center=<3.7668 -0.0317 0.6913>
   rotation=<rot 11.2496 -0.0079 0.091>
  }]
 dyn_hair_rim_l [{shapeType=Box, extentX=5.4008, extentY=14.9559, extentZ=5.4008, center=<-1.3483 0 1.6993>, rotation=<rot -36.7029 0 0>}]
 dyn_hat_rim_r [{shapeType=Box, extentX=5.5905, extentY=20.7167, extentZ=6.3744, center=<0.5079 2.3477 -0.8644>, rotation=<rot 42.7524 0 0>}]
 hat01 [{shapeType=Box, extentX=12.9807, extentY=43.5028, extentZ=39.8513, center=<6.7472 1.141 -0.3284>}]
 hat02 [{shapeType=Box, extentX=8.2376, extentY=16.5711, extentZ=16.9001, center=<3.0067 -1.1129 0.8236>}]
 hat03 [{shapeType=Box, extentX=9.2616, extentY=9.0772, extentZ=5.259, center=<5.0073 -1.4365 -1.1103>}]
 head [{shapeType=Box, extentX=20.7902, extentY=22.082, extentZ=8.3981, center=<1.1305 -0.1949 -0.6824>}]
}
constraints [
 {bone1Name=hat02, bone2Name=hat01, swing1LimitDegrees=20, swing2LimitDegrees=20, twistLimitDegrees=1}
 {bone1Name=hat03, bone2Name=hat02, swing1LimitDegrees=20, swing2LimitDegrees=20, twistLimitDegrees=1}
 {bone1Name=dyn_hair_rim_l, bone2Name=hat01, swing1LimitDegrees=1, swing2LimitDegrees=1, twistLimitDegrees=30}
 {bone1Name=dyn_hat_rim_r, bone2Name=hat01, swing1LimitDegrees=1, swing2LimitDegrees=1, twistLimitDegrees=30}
 {bone1Name=hat01, bone2Name=head, swing1LimitDegrees=3, swing2LimitDegrees=3, twistLimitDegrees=3}
 {bone1Name=dyn_hair_l, bone2Name=head, swing1LimitDegrees=10, swing2LimitDegrees=10, twistLimitDegrees=1}
 {bone1Name=dyn_hair_back, bone2Name=head, swing1LimitDegrees=25, swing2LimitDegrees=25, twistLimitDegrees=1}
 {bone1Name=dyn_hair_r, bone2Name=head, swing1LimitDegrees=40, swing2LimitDegrees=25, twistLimitDegrees=5}
]
