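# Animation Blueprint for the witch hat cosmetic that adds secondary motion, including wobble while skydiving.
# It copies the attached parent's pose, then runs RigidBody physics relative to the head bone with overridden gravity.
# Four rim and hat bones then receive an additive rotation from Perlin noise driven by pawn height, interpolated during skydiving and zeroed otherwise.
# An upward force vector is also computed, halved under parachute, but the AnimGraph leaves it unconnected.
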AnimBlueprint /BRCosmetics/Accessories/Hats/F_MED_Witch_Hat/F_MED_Witch_FaceAcc_ABP
target_skeleton=&/F_MED_Witch_FaceAcc_Skeleton
graphs [
 AnimGraph {nodes [
   n1 = RigidBody(ComponentPose=Local To Component(LocalPose=Copy Pose From Mesh() {bUseAttachedParent=true})) {
    bOverrideWorldGravity=true
    overridePhysicsAsset=&/F_MED_Witch_FaceACC_Physics
    simulationSpace=BaseBoneSpace
    alphaBoolBlend {blendOption=Linear}
    baseBoneRef {boneName=head}
    overrideWorldGravity {z=-980}
    simSpaceSettings {
     maxAngularAcceleration=500
     maxAngularVelocity=3
     maxLinearAcceleration=400
     maxLinearVelocity=250
     velocityScaleZ=0
     worldAlpha=1
     externalLinearDragV=<0.1 0.1>
    }
   }
   n2 = Get current_dive_jiggle() -> (current_dive_jiggle:Rotator)
   n3 = "Transform (Modify) Bone - Bone: dyn_hair_rim_l"(ComponentPose=@n1, Alpha=0.7, Translation="0,0,0", Rotation=@n2, Scale="1,1,1") {rotationMode=BMM_Additive, rotationSpace=BCS_BoneSpace, alphaBoolBlend {blendOption=Linear}, boneToModify {boneName=dyn_hair_rim_l}}
   n4 = "Transform (Modify) Bone - Bone: dyn_hat_rim_r"(ComponentPose=@n3, Alpha=0.7, Translation="0,0,0", Rotation=@n2, Scale="1,1,1") {rotationMode=BMM_Additive, rotationSpace=BCS_BoneSpace, alphaBoolBlend {blendOption=Linear}, boneToModify {boneName=dyn_hat_rim_r}}
   n5 = "Transform (Modify) Bone - Bone: hat02"(ComponentPose=@n4, Alpha=0.7, Translation="0,0,0", Rotation=@n2, Scale="1,1,1") {rotationMode=BMM_Additive, rotationSpace=BCS_BoneSpace, alphaBoolBlend {blendOption=Linear}, boneToModify {boneName=hat02}}
   n6 = "Transform (Modify) Bone - Bone: hat03"(ComponentPose=@n5, Alpha=0.7, Translation="0,0,0", Rotation=@n2, Scale="1,1,1") {rotationMode=BMM_Additive, rotationSpace=BCS_BoneSpace, alphaBoolBlend {blendOption=Linear}, boneToModify {boneName=hat03}}
   Output Pose(Result=Component To Local(ComponentPose=@n6)) {layerGroup=None}
   Break Fort Anim Input Skydiving External Force(FortAnimInput_SkydivingExternalForce=Get SkydivingExternalForceInput()) -> (
    bUseSkydivingVectorForce=~
    SkydivingVectorForce=~
    HeadToPelvisDirection=~
    FloatingMultiplier=~
    FloatingAdditive=~
    DivingMultiplier=~
    DivingAdditive=~
    ParachutingMultiplier=~
    ParachutingAdditive=~
   )
   Get falling_anim_dyn_up_force_vec3() -> (falling_anim_dyn_up_force_vec3:Vector)
  ]}
 EventGraph {nodes [
   n1 = Event BlueprintUpdateAnimation()
   n2 = BreakVector(InVec=Get PawnLastLocation() -> (PawnLastLocation:Vector))
   > Set current_actor_z(current_actor_z:double=@n2.Z) -> (Output_Get:double)
   n3 = MakeRotator(
    Pitch=Make Perlin Noise and Remap!(Value:float=@n2.Z, RangeOutMin:float=-30, RangeOutMax:float=40) -> (ReturnValue:float)
    Yaw=Make Perlin Noise and Remap(Value=float + float(A=@n2.Z, B=20) {is=title K2Node_CommutativeAssociativeBinaryOperator}, RangeOutMin=-30, RangeOutMax=30)
   )
   n4 = Get current_dive_jiggle()
   > Set current_dive_jiggle(current_dive_jiggle:Rotator=SelectRotator!(
     A:Rotator=RInterpTo!(Current:Rotator=@n4, Target:Rotator=@n3, DeltaTime:float=@n1.DeltaTimeX, InterpSpeed:float=10) -> (ReturnValue:Rotator)
     B:Rotator="0, 0, 0"
     bPickA:bool=Get bIsSkyDiving()
    ) -> (ReturnValue:Rotator)) -> (Output_Get:Rotator)
   n5 = SelectFloat(
    A=Make Perlin Noise and Remap(Value=Get current_actor_z() -> (current_actor_z:double), RangeOutMin=0, RangeOutMax=5000)
    B=0
    bPickA=Get bIsSkyDiving()
   )
   n6 = float * float(A=@n5, B=0.5)
   n7 = Get bIsParachuteOpened()
   > Set falling_anim_dyn_up_force_vec3(falling_anim_dyn_up_force_vec3:Vector=MakeVector(Y=0, Z=SelectFloat(A=@n6, B=@n5, bPickA=@n7))) -> (Output_Get:Vector)
   TryGetPawnOwner()
  ]}
]
variables [
 current_actor_z {type=number}
 current_dive_jiggle {type=Rotator}
 falling_anim_dyn_up_force_vec3 {type=Vector}
]
